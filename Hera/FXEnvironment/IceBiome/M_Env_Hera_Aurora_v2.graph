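# Unlit, two-sided translucent aurora material for the Hera ice biome, with a TopDownFacing blend for an alternate look.
# Panning noise and screen-aligned dust drive a curve-atlas color ramp, then hue shift and desaturation.
# Opacity combines UV and curve masks with near-camera and optional mesh fades, plus depth fade on selected quality levels and shading paths.
# Named reroutes organize vertex offsets, sine waves, angle fade and position-based per-component randomness.
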Material /Hera/FXEnvironment/IceBiome/M_Env_Hera_Aurora_v2
graph {
 properties {
  blendMode=BLEND_Translucent
  pixelDepthOffsetMode=PDOM_Legacy
  previewMesh=/Engine/Transient.MaterialEditorMeshComponent_1
  shadingModel=MSM_Unlit
  twoSided=true
  usedShadingModels=Unlit
 }
 nodes [
  Add(A=NamedRerouteUsage(), B=NamedRerouteUsage())
  MaterialFunctionCall() {materialFunction=&AxisAlignedFresnel}
  Multiply(A=Multiply(A=NamedRerouteUsage(), B=NamedRerouteUsage()), B=NamedRerouteUsage())
  n1 = VectorParameter() {parameterName=Param_2, defaultValue=<rgba 0 0 8000 1>}
  n2 = VectorParameter() {parameterName=Param_4, defaultValue=<rgba 0 0 18836.4102 1>}
  n3 = NamedRerouteUsage()
  n4 = TextureCoordinate()
  n5 = Multiply(
   A=Transform(None=LinearInterpolate(A=@n1.RGB, B=@n2.RGB, Alpha=@n3)) {transformSourceType=TRANSFORMSOURCE_Local}
   B=Power(Base=ComponentMask(None=@n4) {g=true}, Exp=ScalarParameter() {defaultValue=1, parameterName=s})
  )
  n6 = VectorParameter() {parameterName=Param_3, defaultValue=<rgba 0 0 -7.9369 0>}
  n7 = Subtract(A=@n6.RGB, B=ScalarParameter() {defaultValue=7.7016, parameterName=Param_15}) {constB=66}
  n8 = NamedRerouteUsage()
  n9 = Multiply(A=ObjectPositionWS(), B=Divide(A=LinearInterpolate(A=@n6.RGB, B=@n7, Alpha=@n8)) {constB=100})
  n10 = Transform(None=LinearInterpolate(
    A=VectorParameter() {defaultValue=<rgba 0 6236.606 0 0>}.RGB
    B=Constant3Vector() {constant=<rgba 0 3240.5295 0 0>}
    Alpha=NamedRerouteUsage()
   ))
  MaterialExpressionClamp_2 = Clamp(None=ConstantBiasScale(None=TextureCoordinate()) {bias=-0.5, scale=2}) {minDefault=-1}
  n11 = Convert(None=@MaterialExpressionClamp_2) {
   nodeName=Break Float2
   convertInputs [{
     type=Vector2
     expressionInput {
      expression=&self:MaterialExpressionClamp_2
      inputName=None
      mask=0
      maskA=0
      maskB=0
      maskG=0
      maskR=0
      outputIndex=0
     }
    }]
   convertOutputs [{type=Scalar}, {type=Scalar}]
  }
  NamedRerouteDeclaration(None=LinearInterpolate(A=Add(A=@n5, B=@n9), B=Multiply(A=@n10, B=@n11), Alpha=ScalarParameter() {parameterName=TopDownFacing?})) {name=WPO, nodeColor=<rgba 1 0.0292 0.6744 1>}
  n12 = ScalarParameter() {defaultValue=0.05, parameterName=Param_17}
  n13 = NamedRerouteUsage()
  n14 = MaterialFunctionCall(Float2=TextureCoordinate()) {materialFunction=&BreakOutFloat2Components}
  n15 = CurveAtlasRowParameter(CurveTime=LinearInterpolate(A=@n14.R, B=OneMinus(None=@n14.R), Alpha=ScalarParameter() {group=Curve, parameterName=InvetColorRamp})) {atlas=&/CA_Env_Hera_Aurora, curve=&/C_Env_Hera_Aurora, parameterName=AuroraColor}
  n16 = Multiply(A=@n15) {constB=0.5}
  n17 = Constant() {r=256}
  n18 = Constant2Vector() {g=0.05}
  n19 = Time()
  n20 = TextureSample(UVs=Add(
    A=MaterialFunctionCall!(TextureResolution=@n17) {materialFunction=/Engine/Functions/Engine_MaterialFunctions02/Texturing/ScreenAlignedPixelToPixelUVs}.Result
    B=Multiply(A=@n18, B=@n19)
   )) {texture=/Game/Global/Textures/Default/PointSplatter/T_Tiling_Dust}
  n21 = Multiply(A=@n20.R, B=@n20.R)
  n22 = TextureCoordinate()
  n23 = TextureCoordinate() {uTiling=0.5, vTiling=0.5}
  n24 = LinearInterpolate(A=ComponentMask(None=OneMinus(None=@n22)) {g=true}, B=Sine(None=ComponentMask(None=@n23) {g=true}), Alpha=Constant()) {constB=0}
  n25 = CurveAtlasRowParameter(CurveTime=OneMinus(None=@n14.G)) {
   atlas=/Game/Athena/LinearColorCurves/VFX_CurveAtlas
   curve=/Game/Accessories/FORT_Backpacks/Backpack_F_MED_Quarrel/FX/Curve_Quarrel_Auroras_03
   defaultValue=30
   group=Curve
   parameterName=MaskRamp2
  }
  n26 = Multiply(A=@n24, B=@n24)
  n27 = TextureCoordinate()
  n28 = AppendVector(A=Add(A=Constant() {r=4}, B=ScalarParameter() {defaultValue=4.3062, parameterName=Param_16}), B=Constant() {r=0.1})
  n29 = Constant() {r=0.5}
  n30 = MaterialFunctionCall(
   UVs=LinearInterpolate(A=TextureCoordinate() {uTiling=4, vTiling=0.1}, B=Multiply(A=@n27, B=@n28), Alpha=Constant())
   "Rotation Angle (0-1)"=@n29
  ) {materialFunction=&CustomRotator}
  n31 = ScalarParameter() {defaultValue=0.1, parameterName=Param_12}
  n32 = Multiply(A=Time(), B=LinearInterpolate(A=@n31, B=Add(A=@n31) {constB=0.4}, Alpha=NamedRerouteUsage()) {constB=0.7})
  n33 = MaterialFunctionCall(UVs=TextureCoordinate() {uTiling=2, vTiling=0.1}, "Rotation Angle (0-1)"=@n29) {materialFunction=&CustomRotator}
  n34 = Multiply(
   A=TextureSample(UVs=Panner(Coordinate=@n30.Rotated Values, Time=@n32) {speedX=0.2}) {texture=/Game/Effects/Fort_Effects/Textures/Power_Source/T_Power_Source_Sphere_Texture}.R
   B=TextureSample(UVs=Panner(Coordinate=@n33.Rotated Values, Time=@n32) {speedX=-0.05}) {texture=&T_Inky_Smoke_Tile}.R
  )
  n35 = ScalarParameter() {defaultValue=2, parameterName=TexMultiply}
  n36 = NamedRerouteUsage()
  n37 = Add(
   A=Multiply(A=@n25.A, B=ScalarParameter() {defaultValue=0.7, parameterName=BottomGlow}) {constB=0.5}
   B=Multiply(A=Multiply(A=@n26, B=@n34), B=LinearInterpolate(B=@n35, Alpha=@n36) {constA=1.5}) {constB=5}
  )
  n38 = ScalarParameter() {defaultValue=0.2, parameterName=ColorAdd}
  n39 = Add(A=Multiply(A=Multiply(A=@n21) {constB=7}, B=@n24), B=Saturate(None=Add(A=@n37, B=@n38) {constB=0.05}))
  n40 = CurveAtlasRowParameter(CurveTime=@n39) {atlas=&CA_Env_Hera_Aurora, curve=&C_Env_Hera_Aurora, parameterName=AuroraColor}
  n41 = LinearInterpolate(
   A=MaterialFunctionCall(Hue Shift Percentage=LinearInterpolate(B=@n12, Alpha=@n13), Texture=Add(A=@n16, B=@n40)) {materialFunction=&HueShift}.Result
   B=VectorParameter() {parameterName=Param_6, defaultValue=<rgba 0.7167 0.0986 2 1>}.RGB
   Alpha=ScalarParameter() {parameterName=TopDownFacing?}
  )
  n42 = LinearInterpolate(B=ScalarParameter() {defaultValue=0.2, parameterName=Param_18}, Alpha=NamedRerouteUsage())
  NamedRerouteDeclaration(None=MaterialFunctionCall(Emissive=Desaturation(None=@n41, Fraction=@n42)) {materialFunction=&MF_FX_EmissiveAdjustments}.Result) {name=Color, nodeColor=<rgba 1 0 0.3047 1>}
  n43 = ScalarParameter() {parameterName=Param_21}
  n44 = ScalarParameter() {defaultValue=0.8, parameterName=Param_22}
  n45 = CameraVectorWS()
  n46 = Abs(None=DotProduct(A=Normalize(VectorInput=@n45), B=VectorParameter() {parameterName=Param_1, defaultValue=<rgba 0 0 1 1>}.RGB))
  n47 = OneMinus(None=@n46)
  n48 = ScalarParameter() {parameterName=Param_3}
  n49 = ScalarParameter() {defaultValue=1, parameterName=Param_8}
  NamedRerouteDeclaration(None=LinearInterpolate(
    A=SmoothStep(Min=@n43, Max=@n44, Value=@n47) {constMin=0.4}
    B=SmoothStep(Min=@n48, Max=@n49, Value=@n46) {constMin=0.4}
    Alpha=ScalarParameter() {parameterName=TopDownFacing?}
   )) {name=AngelFade, nodeColor=<rgba 0.9922 0 1 1>}
  n50 = CurveAtlasRowParameter(CurveTime=@n14.R) {
   atlas=&VFX_CurveAtlas
   curve=&Curve_Quarrel_Auroras_03
   defaultValue=30
   group=Curve
   parameterName=MaskRamp1
  }
  n51 = ComponentMask(None=TextureCoordinate()) {g=true}
  n52 = Multiply(A=Power(Base=OneMinus(None=@n51)), B=ScalarParameter() {defaultValue=2, parameterName=Param_14}) {constB=0.7}
  n53 = Add(A=TextureCoordinate()) {constB=0.25}
  n54 = OneMinus(None=Sine(None=ComponentMask(None=@n53) {g=true}))
  n55 = Multiply(A=@n39, B=Multiply(A=@n50.B, B=Multiply(A=@n52, B=@n54)))
  n56 = NamedRerouteUsage()
  n57 = ScalarParameter() {parameterName=TopDownFacing?}
  n58 = ScalarParameter() {defaultValue=0.4, parameterName=FinalMaskMultiply}
  n59 = NamedRerouteUsage()
  n60 = Multiply(
   A=Multiply(A=LinearInterpolate(A=@n55, B=@n56, Alpha=@n57), B=LinearInterpolate(A=@n58, Alpha=@n59) {constB=0.3}) {constB=2}
   B=MaterialFunctionCall(Distance=Constant() {r=64}) {materialFunction=&Near_Camera_Fade}.Result
  )
  n61 = Abs(None=DotProduct(A=CameraVectorWS(), B=VertexNormalWS()))
  n62 = Multiply(A=@n61, B=@n61)
  n63 = LinearInterpolate(A=@n60, B=Multiply(A=@n60, B=Saturate(None=@n62)), Alpha=ScalarParameter() {parameterName=Use MeshFade})
  n64 = DepthFade(Opacity=@n63, FadeDistance=ScalarParameter() {parameterName=FadeDistance})
  n65 = QualitySwitch(Default=@n64, Low=@n63, High=@n64, Medium=@n63, Epic=@n64)
  NamedRerouteDeclaration(None=ShadingPathSwitch(Default=@n65, Deferred=@n65, Forward=@n63, Mobile=@n63)) {name=Opacity, nodeColor=<rgba 1 0.0938 0 1>}
  n66 = VectorParameter() {parameterName="Size&&SineWave", channelNames {a=UpDown, b=SideToSide, g=Width, r=Length}, defaultValue=<rgba 100 1456.4949 2649.5442 0>}
  n67 = VertexNormalWS()
  n68 = Transform(None=Constant3Vector() {constant=<rgba 0 1 0 0>})
  n69 = ComponentMask(None=TextureCoordinate()) {r=true}
  n70 = VectorParameter() {parameterName="UpDown&&FrontBack", channelNames {a=Pan, b=Tiling, g=Pan, r=Tiling}, defaultValue=<rgba 1 1 1 1>}
  n71 = Time()
  n72 = ComponentMask(None=Frac(None=ObjectPositionWS())) {g=true}
  n73 = Multiply(
   A=@n71
   B=LinearInterpolate(A=ScalarParameter() {defaultValue=0.03}, Alpha=NamedRerouteDeclaration(None=@n72) {name=PerComponentRandom, nodeColor=<rgba 1 0 0.4453 1>}) {constB=0.04}
  )
  n74 = Add(A=Add(A=Multiply(A=@n69, B=@n70.Tiling), B=Multiply(A=@n70.Pan, B=@n73)), B=ScalarParameter() {parameterName=PanOffset1})
  n75 = Multiply(A=LinearInterpolate(A=@n67, B=@n68), B=Cosine(None=@n74))
  n76 = ScalarParameter() {parameterName=TopDownFacing?}
  n77 = Add(A=Add(A=Multiply(A=@n69, B=@n70.Tiling), B=Multiply(A=@n70.Pan, B=@n71)), B=ScalarParameter() {parameterName=PanOffset2})
  n78 = Multiply(A=@n66.UpDown, B=Multiply(A=LinearInterpolate(A=@n68, B=@n67, Alpha=@n76), B=Sine(None=@n77))) {constB=20}
  n79 = TextureCoordinate() {uTiling=0.5, vTiling=0.5}
  n80 = Saturate(None=Sine(None=ComponentMask(None=@n79) {r=true}))
  NamedRerouteDeclaration(None=Add(A=Multiply(A=@n66.SideToSide, B=@n75) {constB=20}, B=Multiply(A=@n78, B=@n80) {constB=20})) {name=SineWave, nodeColor=<rgba 1 0.7031 0 1>}
  n81 = SphereMask(
   A=ComponentMask(None=TextureCoordinate()) {g=true}
   B=ScalarParameter() {defaultValue=0.5, parameterName=Param_23}
   Radius=ScalarParameter() {defaultValue=0.5, parameterName=Param_24}
   Hardness=ScalarParameter() {parameterName=Param_25}
  )
  n82 = Sine(None=ComponentMask(None=TextureCoordinate() {uTiling=0.5}) {r=true})
  NamedRerouteDeclaration(None=Power(Base=Multiply(A=@n81, B=@n82), Exp=ScalarParameter() {defaultValue=2.3466, parameterName=Param_26})) {name=GradientMask, nodeColor=<rgba 1 0.0469 0 1>}
  n83 = ScalarParameter() {defaultValue=2.1241, parameterName=Param_20}
  NamedRerouteDeclaration(None=OneMinus(None=Fresnel(ExponentIn=@n83))) {name=Fresnel, nodeColor=<rgba 0.6875 1 0 1>}
  NamedRerouteUsage()
  ScalarParameter() {parameterName=TopDownFacing?}
  ScalarParameter() {parameterName=TopDownFacing?}
  Sine(None=@n74)
 ]
}
